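# CentralGust skydiving contrail Niagara system: wind-gust trails with twin tornado meshes, ribbons, sparks, a shockwave and swirling debris (leaves, sticks, occasional "Lama" meshes).
# Contrail setup and system scripts supply socket readers and direction axes that emitters use for placement and motion.
# Ribbons are static bursts repositioned along those axes each frame; debris swirls through sine/cosine forces.
# ParaGlideLeanAlpha scales rates, speed, stretch and shockwave opacity; front-end preview retunes Lama motion.
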NiagaraSystem /BRCosmetics/Athena/Effects/Particles/Skydiving/NS_Contrail_CentralGust/NS_Contrail_CentralGust
dependencies {
 usedDataInterfaces [
  &NiagaraDataInterfaceCurve
  &NiagaraDataInterfaceCamera
  &NiagaraDataInterfaceMeshRendererInfo
  &NiagaraDataInterfaceArrayColor
  &NiagaraDataInterfaceSocketReader
  &NiagaraDataInterfaceSkeletalMesh
 ]
 usedDynamicInputs []
 usedModules [
  &self:SystemSpawnScript.SystemScriptSource.SystemScriptGraph.NiagaraNodeAssignment_1.SetVariables_#1
  &self:SystemSpawnScript.SystemScriptSource.SystemScriptGraph.NiagaraNodeAssignment_0.SetVariables_#2
  &SystemState
  /Game/Effects/Niagara/Modules/Contrails/FN_Contrail_Setup
  &FN_Integrator
  &self:ShockWaveOrientVector
  &self:SetDiveDir
 ]
}
emitters {
 Lama {
  emitterUpdateScript [
   &EmitterState()
   &SpawnRate(
    Spawn Probability=&Multiply_Float(A=$User.ParaGlideLeanAlpha, B=0.75)
    SpawnGroup=1
    SpawnRate=0.4
    Use Spawn Probability=true
   ) {module=SpawnRate001}
  ]
  particleSpawnScript [
   /SetVariables(
    Particles.Lama_Age_Scale=0.7
    Particles.Lama_Drag:NiagaraFloat=&MakeCustomFloatFromBool(Bool=$User.bIsFrontEndPreview, True Float=0.5, False Float=2)
    Particles.Lama_Force_Scale_A=100
    Particles.Lama_Force_Scale_X:NiagaraFloat=&MakeCustomFloatFromBool(Bool=$User.bIsFrontEndPreview, True Float=300, False Float=300)
    Particles.Lama_Force_Scale_Y:NiagaraFloat=&MakeCustomFloatFromBool(Bool=$User.bIsFrontEndPreview, True Float=300, False Float=300)
    Particles.Lama_LifeTime:NiagaraFloat=&RandomRangeFloat(Minimum=3, Maximum=5)
    Particles.Lama_Rotation_Rate:NiagaraFloat=&RandomRangeFloat(Minimum=0.5, Maximum=1.2)
    Particles.Lama_Scale:NiagaraFloat=&MakeCustomFloatFromBool(Bool=$User.bIsFrontEndPreview, True Float=0.35, False Float=0.25)
    Particles.Lama_Velocity_Speed:NiagaraFloat=&MakeCustomFloatFromBool(Bool=$User.bIsFrontEndPreview, True Float=100, False Float=350)
   ) {isSetParameters=true, module=SetVariables_#3}
   &PartitionParticles()
   &InitializeParticle(
    Color Mode=::NewEnumerator0
    Lifetime=$Particles.Lama_LifeTime
    Mesh Scale Mode=::NewEnumerator3
    Mesh Uniform Scale=$Particles.Lama_Scale
    UsePositionOffset=false
    Write Material Random=true
   )
   &SocketLocation(
    Orientation Sampling=::NewEnumerator1
    SamplePosOffset=false
    Socket Index=&ModuloInt(A=&ReturnParticleIDAttributes(), B=2)
    Socket Reader=$System.Socket Reader
    Socket Sampling Mode=::NewEnumerator3
    VelocitySpeedLimit=false
   )
   &AddVelocity(
    Velocity=&SelectVectorFromArray(
     Direct Array Index=/Niagara/DynamicInputs/Subtract/Subtract_Integer(
      A:NiagaraInt32=1
      B:NiagaraInt32=&ModuloInt(A=&ReturnParticleIDAttributes(), B=2)
     )
     Vector Selection Array=$System.Axis
    )
    Velocity Speed Scale=$Particles.Lama_Velocity_Speed
   )
  ]
  particleUpdateScript [
   &ParticleState()
   &LinearForce(Force=&Add_Vector(
     A=&MakeVector(
      X=&Sine(
       Normalized Angle=&Multiply_Float(A=$Emitter.Age, B=$Particles.Lama_Age_Scale)
       Period=1
       Scale=$Particles.Lama_Force_Scale_X
       Bias=0
      )
      Y=&Cosine(
       Normalized Angle=&Multiply_Float(A=$Emitter.Age, B=$Particles.Lama_Age_Scale)
       Period=1
       Scale=$Particles.Lama_Force_Scale_Y
       Bias=0
      )
      Z=0
     )
     B=&Multiply_VectorByFloat(
      Vector=&SelectVectorFromArray(
       Direct Array Index=&Subtract_Integer(
        A:NiagaraInt32=1
        B:NiagaraInt32=&ModuloInt(A=&ReturnParticleIDAttributes(), B=2)
       )
       Vector Selection Array=$System.Axis
      )
      Float=$Particles.Lama_Force_Scale_A
     )
    ))
   &Drag(Drag=$Particles.Lama_Drag, Use Linear Drag=true, Use Rotational Drag=true)
   &UpdateMeshOrientation(Rotation Rate=$Particles.Lama_Rotation_Rate, Rotation Vector=&RandomVector(Vector Scale=1))
   &ScaleMeshSize(Scale Factor=&VectorFromFloat(Value=&FloatFromCurve(FloatCurve=DataInterface() {propertyValues {Curve {
         defaultValue=1
         keys [
          {interpMode=RCIM_Cubic, leaveTangent=42.399, tangentMode=RCTM_Break, time=0, value=0}
          {arriveTangent=0.0121, interpMode=RCIM_Cubic, leaveTangent=0.0121, tangentMode=RCTM_User, time=0.1, value=1}
          {arriveTangent=-0.0405, interpMode=RCIM_Cubic, leaveTangent=-0.0405, tangentMode=RCTM_User, time=0.7009, value=1}
          {arriveTangent=-8.0089, interpMode=RCIM_Cubic, leaveTangent=-8.0089, tangentMode=RCTM_User, time=1, value=0}
         ]
        }}})))
   &SolveForcesAndVelocity(Clamp Velocity=false, Limit Acceleration=false)
  ]
  rendererClasses [&NiagaraMeshRendererProperties]
  renderers [{rendererClass=&NiagaraMeshRendererProperties, rendererIndex=0}]
 }
 Leafs {
  emitterUpdateScript [
   &EmitterState()
   &SpawnRate(SpawnRate=&Lerp_Float(A=40, B=80, Alpha=$User.ParaGlideLeanAlpha), Use Spawn Probability=false) {module=SpawnRate003}
  ]
  particleSpawnScript [
   &InitializeParticle(
    Color=&SelectLinearColorFromArray(
     Array Sampling Mode=::NewEnumerator1
     Direct Array Index=&RandomRangeInt(Minimum=0, Maximum=3)
     Color Selection Array=DataInterface() {propertyValues {ColorData [<rgba 0.3003 0.1432 0 1>, <rgba 0.2288 0.3177 0 1>, <rgba 0.3425 0.3594 0 1>, <rgba 0.2865 0.2354 0 1>]}}
     Random Seed=?
    )
    Lifetime Max=0.8
    Lifetime Min=0.6
    Lifetime Mode=::NewEnumerator1
    Mass=&MaskFloat(
     Mask Source=0
     Mask Parameter=?
     Mask=0
     Masked Float=&RandomRangeFloat(Minimum=0.8, Maximum=1.2)
     Passthrough Float=&RandomRangeFloat(Minimum=2, Maximum=3)
    )
    Mass Mode=::NewEnumerator1
    Mesh Scale Mode=::NewEnumerator3
    Mesh Uniform Scale=&MaskFloat(
     Mask Source=0
     Mask Parameter=?
     Mask=0
     Masked Float=&RandomRangeFloat(Minimum=0.1, Maximum=0.3)
     Passthrough Float=&RandomRangeFloat(Minimum=1, Maximum=3)
    )
    Sprite Rotation Mode=::NewEnumerator2
    UsePositionOffset=false
    Write Material Random=true
   )
   /SetVariables(
    Particles.RadScale:NiagaraFloat=&RandomRangeFloat(Minimum=2500, Maximum=2500)
    Particles.RandomOffsetAngle:NiagaraFloat=&RandomRangeFloat(Minimum=0, Maximum=1)
    Particles.SpeedScale:NiagaraFloat=&RandomRangeFloat(Minimum=2, Maximum=2)
   ) {isSetParameters=true}
   /SetVariables(Particles.MeshIndex=&RandomRangeInt(Minimum=0, Maximum=3)) {isSetParameters=true}
   &PartitionParticles(Partition Mode:ENiagara_PartitionMode_v2=::NewEnumerator5)
   &SkeletalMeshLocation(
    Bone / Socket Index=$Particles.PartitionParticles.Partition
    Interpolate Mesh Sampling=false
    MaskSpawnGroup=false
    Mesh Sampling Type=::NewEnumerator5
    Orientation Sampling=::NewEnumerator1
    Sampled Position Offset=&Multiply_VectorByFloat(
     Vector=&NormalizeVector(
      Vector To Normalize=&RandomRangeVector(Minimum=<-1 -1 -1>, Maximum=<1 1 1>)
      Safe Normalize=true
      Safe Fallback Threshold=0
      Safe Fallback Vector=<0 0 0.001>
     )
     Float=&MaskFloat(
      Mask Source=0
      Mask Parameter=?
      Mask=0
      Masked Float=&RandomRangeFloat(Minimum=2, Maximum=5)
      Passthrough Float=&RandomRangeFloat(Minimum=2, Maximum=5)
     )
    )
    Skeletal Mesh=$System.Skeletal Mesh
    Socket Sampling Mode=::NewEnumerator5
   )
   &AddVelocity(
    Cone Angle=&Lerp_Float(A=45, B=5, Alpha=$User.ParaGlideLeanAlpha)
    Cone Axis=&SelectVectorFromArray(Direct Array Index=$Particles.PartitionParticles.Partition, Vector Selection Array=$System.Axis)
    Module.UseConeFalloff=false
    Module.UseVelDistribution=false
    Override Seed=false
    Velocity Mode=::NewEnumerator2
    Velocity Speed=&Lerp_Float(
     A=&RandomRangeFloat(Minimum=150, Maximum=250)
     B=&RandomRangeFloat(Minimum=600, Maximum=450)
     Alpha=$User.ParaGlideLeanAlpha
    )
   ) {module=AddVelocity001}
   &InitialMeshOrientation(
    Mesh Orientation Mode=::NewEnumerator1
    Rotation=&RandomRangeVector(Minimum=<-1 -1 -1>, Maximum=<1 1 1>)
    Rotation Coordinate Space=::NewEnumerator2
    Use Rotation Vector=true
   )
  ]
  particleUpdateScript [
   &ParticleState()
   &LinearForce(Force=&MakeVector(
     X=&Sine(
      Normalized Angle=&Add_Float(A=&Multiply_Float(A=$Emitter.Age, B=$Particles.SpeedScale), B=$Particles.RandomOffsetAngle)
      Period=1
      Scale=&Multiply_Float(A=$Particles.RadScale, B=$Particles.NormalizedAge)
      Bias=0
     )
     Y=&Cosine(
      Normalized Angle=&Add_Float(A=&Multiply_Float(A=$Emitter.Age, B=$Particles.SpeedScale), B=$Particles.RandomOffsetAngle)
      Period=1
      Scale=&Multiply_Float(A=$Particles.RadScale, B=$Particles.NormalizedAge)
      Bias=0
     )
     Z=0
    ))
   &AerodynamicDrag(Aerodynamic Drag=0.1, Rotation Strength=20, Use Linear Drag=true, Use Rotational Drag=true)
   &SolveForcesAndVelocity(Clamp Velocity=false, Limit Acceleration=false)
   &AlignSpriteToMeshOrientation(SetAlignment=true, SetFacing=true)
   &ScaleMeshSize(Scale Factor=&VectorFromFloat(Value=&FloatFromCurve(FloatCurve=DataInterface() {propertyValues {Curve {
         defaultValue=1
         keys [
          {tangentMode=RCTM_User, time=0, value=0}
          <at 0.1119 1>
          {arriveTangent=-0.0405, interpMode=RCIM_Cubic, leaveTangent=-0.0405, tangentMode=RCTM_User, time=0.9, value=1}
          {arriveTangent=-24.0426, interpMode=RCIM_Cubic, leaveTangent=-24.0426, tangentMode=RCTM_User, time=1, value=0}
         ]
        }}})))
  ]
  rendererClasses [&NiagaraMeshRendererProperties]
  renderers [{rendererClass=&NiagaraMeshRendererProperties, rendererIndex=0}]
 }
 ParticletoRibbon {
  emitterUpdateScript [
   &EmitterState()
   &SpawnBurst_Instantaneous(Spawn Count=40, Use Spawn Probability=false, UseLoopCountLimit=true)
  ]
  particleSpawnScript [
   &PartitionParticles(
    Normalized Index In Partition Includes One=true
    Particles Per Partition=10
    Partition Mode:ENiagara_PartitionMode_v2=::NewEnumerator1
    Write Partition Index to Ribbon Link Order:ENiagara_AttributeSamplingApplyOutput=::NewEnumerator0
    Write Partition to Ribbon ID=::NewEnumerator0
   )
   &InitializeParticle(
    Color=&FNTimeOfDayColor(Night Color=<rgba 0.5517 0.7251 1 0.7>, Morning Color=<rgba 1 1 1 0.5>, Day Color=<rgba 1 1 1 1>, Evening Color=<rgba 1 0.8396 0.7176 1>)
    Lifetime=0.2
    Ribbon Width=1
    Ribbon Width Mode=::NewEnumerator1
    Sprite Size Mode=::NewEnumerator3
    UsePositionOffset=false
    Write Material Random=true
   )
  ]
  particleUpdateScript [
   &ParticleState(Kill Particles When Lifetime Has Elapsed=false)
   &SkeletalMeshLocation(
    Bone / Socket Index=$Particles.PartitionParticles.Partition
    MaskSpawnGroup=false
    Mesh Sampling Type=::NewEnumerator5
    Orientation Sampling=::NewEnumerator1
    Skeletal Mesh=$System.Skeletal Mesh
    Socket Sampling Mode=::NewEnumerator5
   )
   /SetVariables(Particles.CustomCameraFade:NiagaraFloat=&FloatFromCurve(
     FloatCurve=DataInterface() {propertyValues {Curve {keys [<at 0 1>,<at 1 0>]}}}
     CurveIndex=&self:CustomCameraFadeIndex(Camera Query:NiagaraDataInterfaceCamera=DataInterface() {propertyValues {bRequireCurrentFrameData=false}})
    )) {isSetParameters=true}
   &ScaleColor(
    Scale Alpha=&FloatFromCurve(
     FloatCurve=DataInterface() {propertyValues {Curve {
        defaultValue=0
        keys [
         <at 0 0 RCIM_Cubic>
         {arriveTangent=0.0981, interpMode=RCIM_Cubic, leaveTangent=0.0981, tangentMode=RCTM_User, time=0.01, value=1}
         {interpMode=RCIM_Cubic, tangentMode=RCTM_Break, time=0.25, value=1}
         {arriveTangent=-3.5776, interpMode=RCIM_Cubic, leaveTangent=-3.5776, tangentMode=RCTM_User, time=0.35, value=0.5}
         {interpMode=RCIM_Cubic, tangentMode=RCTM_Break, time=0.85, value=0}
        ]
       }}}
     CurveIndex=$Particles.PartitionParticles.Normalized Index In Partition
     Scale Curve=&Multiply_Float(A=$Particles.CustomCameraFade, B=0.35)
    )
    Scale RGB=&VectorFromFloat(Value=0.5)
    ScaleA=true
    ScaleRGB=true
   )
   &ScaleRibbonWidth(Ribbon Width Scale=&FloatFromCurve(
     FloatCurve=DataInterface() {propertyValues {Curve {
        defaultValue=0
        keys [
         {arriveTangent=0.2503, interpMode=RCIM_Cubic, leaveTangent=0.2503, tangentMode=RCTM_User, time=0, value=0.15}
         {arriveTangent=1.5693, interpMode=RCIM_Cubic, leaveTangent=1.5693, tangentMode=RCTM_User, time=0.35, value=0.65}
         {interpMode=RCIM_Cubic, tangentMode=RCTM_User, time=1, value=1}
        ]
       }}}
     CurveIndex=$Particles.PartitionParticles.Normalized Index In Partition
     Scale Curve=100
    ))
   &SolveForcesAndVelocity(Clamp Velocity=false, Limit Acceleration=false)
   &DynamicMaterialParameters(
    Index 1 Param 1=&Multiply_Float(A=$Emitter.Age, B=-1)
    Index 1 Param 2=1
    Index 1 Param 3=0.25
    Index 1 Param 4=0.5
    Module.Index0Float0Write=true
    Module.Index0Float1Write=true
    Module.Index0Float2Write=true
    Module.Index0Float3Write=true
    Module.Index1Float0Write=true
    Module.Index1Float1Write=true
    Module.Index1Float2Write=true
    Module.Index1Float3Write=true
    Write Parameter Index 1=::NewEnumerator1
   )
   /SetVariables(Particles.Position=&AddVectorToPosition(
     Position=$Particles.Position
     Vector=&Multiply_VectorByFloat(
      Vector=&Multiply_VectorByFloat(
       Vector=&SelectVectorFromArray(Direct Array Index=$Particles.PartitionParticles.Partition, Vector Selection Array=$System.Axis)
       Float=&Lerp_Float(A=220, B=350, Alpha=$User.ParaGlideLeanAlpha)
      )
      Float=$Particles.PartitionParticles.Normalized Index In Partition
     )
    )) {isSetParameters=true}
  ]
  rendererClasses [&NiagaraRibbonRendererProperties]
  renderers [{rendererClass=&NiagaraRibbonRendererProperties}]
 }
 ShockWave {
  emitterUpdateScript [
   &EmitterState(Life Cycle Mode=::NewEnumerator1, Loop Behavior=::NewEnumerator1, Loop Duration Mode=::NewEnumerator1, UseLoopDelay=false)
   &SpawnBurst_Instantaneous(Use Spawn Probability=false, UseLoopCountLimit=false)
  ]
  particleSpawnScript [&InitializeParticle(
    Color=&FNTimeOfDayColor(Night Color=<rgba 0.5517 0.7251 1 0.7>, Morning Color=<rgba 1 1 1 0.5>, Day Color=<rgba 1 1 1 1>, Evening Color=<rgba 1 0.8396 0.7176 1>)
    Lifetime=5
    Mesh Scale=<3 1 1>
    Mesh Scale Mode=::NewEnumerator4
    Sprite Size Mode=::NewEnumerator3
    Uniform Sprite Size=19.3013
    UsePositionOffset=false
    Write Material Random=true
   )]
  particleUpdateScript [
   &SkeletalMeshLocation(
    Bone / Socket Index=5
    Bone Sampling Mode=::NewEnumerator4
    MaskSpawnGroup=false
    Orientation Sampling=::NewEnumerator1
    Skeletal Mesh=$System.Skeletal Mesh
    Use Sampled Mesh Attribute as Ribbon ID=false
   )
   &ParticleState(Kill Particles When Lifetime Has Elapsed=false, Let Infinitely Lived Particles Die When Emitter Deactivates=true)
   &UpdateMeshOrientation(
    Facing Direction=$System.OrientVector
    Facing Interpolation=false
    Orientation Method=::NewEnumerator1
    Override Quat=false
    TransformFacing=false
    TransformUp=false
   )
   &ScaleColor(
    Scale Alpha=&FloatFromCurve(
     FloatCurve=DataInterface() {propertyValues {Curve {
        defaultValue=1
        keys [
         {arriveTangent=7.3908, interpMode=RCIM_Cubic, leaveTangent=7.3908, tangentMode=RCTM_User, time=0.7, value=0}
         {interpMode=RCIM_Cubic, tangentMode=RCTM_User, time=1, value=1}
        ]
       }}}
     CurveIndex=$User.ParaGlideLeanAlpha
    )
    ScaleA=true
    ScaleRGB=true
   )
  ]
  rendererClasses [&NiagaraMeshRendererProperties]
  renderers [{rendererClass=&NiagaraMeshRendererProperties, rendererIndex=0}]
 }
 Sparks {
  emitterUpdateScript [
   &EmitterState()
   &SpawnRate(SpawnRate=50, Use Spawn Probability=false)
  ]
  particleSpawnScript [
   &InitializeParticle(
    Color=&FNTimeOfDayColor(Night Color=<rgba 0.5517 0.7251 1 0.7>, Morning Color=<rgba 1 1 1 0.5>, Day Color=<rgba 1 1 1 1>, Evening Color=<rgba 1 0.8396 0.7176 1>)
    Lifetime Max=1
    Lifetime Min=0.25
    Lifetime Mode=::NewEnumerator1
    Sprite Size=&Multiply_Vector2D_ByFloat(
     Float=&RandomRangeFloat(Minimum=1, Maximum=3)
     Vector2D=&MakeVector2D(X=0.5, Y=&Lerp_Float(A=4, B=8, Alpha=$User.ParaGlideLeanAlpha))
    )
    Sprite Size Mode=::NewEnumerator4
    UsePositionOffset=false
    Write Material Random=true
   )
   /SetVariables(Particles.SyncIntVar:NiagaraInt32=&RandomRangeInt(Minimum=0, Maximum=1)) {isSetParameters=true}
   &SkeletalMeshLocation(
    Bone / Socket Index=$Particles.SyncIntVar
    MaskSpawnGroup=false
    Mesh Sampling Type=::NewEnumerator5
    Orientation Sampling=::NewEnumerator1
    Sampled Position Offset=&RandomVector(Vector Scale=10)
    Skeletal Mesh=$System.Skeletal Mesh
    Socket Sampling Mode=::NewEnumerator5
    Use Sampled Mesh Attribute as Ribbon ID=false
   )
   &AddVelocity(
    Cone Angle=&RandomRangeFloat(Minimum=10, Maximum=35)
    Cone Axis=&SelectVectorFromArray(Array Sampling Mode=::NewEnumerator0, Direct Array Index=?, Vector Selection Array=$System.Axis, Random Seed=0)
    Module.UseConeFalloff=false
    Module.UseVelDistribution=false
    Override Seed=false
    Velocity Mode=::NewEnumerator2
    Velocity Speed=&Multiply_Float(
     A=&RandomRangeFloat(Minimum=125, Maximum=250)
     B=&Lerp_Float(A=1, B=1.5, Alpha=$User.ParaGlideLeanAlpha)
    )
   )
  ]
  particleUpdateScript [
   &ParticleState()
   &LinearForce(Force=&MakeVector(
     X=&Sine(Normalized Angle=&Multiply_Float(A=$Emitter.Age, B=2), Period=1, Scale=500, Bias=0)
     Y=&Cosine(Normalized Angle=&Multiply_Float(A=$Emitter.Age, B=2), Period=1, Scale=500, Bias=0)
     Z=0
    ))
   &Drag(Use Linear Drag=true, Use Rotational Drag=false)
   &SolveForcesAndVelocity(Clamp Velocity=false, Limit Acceleration=false)
   &ScaleSpriteSize(Uniform Curve Sprite Scale=DataInterface() {propertyValues {Curve {
       defaultValue=0
       keys [
        {tangentMode=RCTM_User, time=0, value=0}
        <at 0.1 1>
        {arriveTangent=-8.0089, interpMode=RCIM_Cubic, leaveTangent=-8.0089, tangentMode=RCTM_User, time=1, value=0}
       ]
      }}})
   &ScaleColor(
    Scale Alpha=&FloatFromCurve(
     FloatCurve=DataInterface() {propertyValues {Curve {
        defaultValue=1
        keys [{interpMode=RCIM_Cubic, tangentMode=RCTM_User, time=0, value=1}, {interpMode=RCIM_Cubic, tangentMode=RCTM_User, time=1, value=0}]
       }}}
     Scale Curve=0.5
    )
    ScaleA=true
    ScaleRGB=false
   )
  ]
  rendererClasses [&NiagaraSpriteRendererProperties]
  renderers [{rendererClass=&NiagaraSpriteRendererProperties}]
 }
 Sticks {
  emitterUpdateScript [
   &EmitterState()
   &SpawnRate(SpawnRate=2, Use Spawn Probability=false)
  ]
  particleSpawnScript [
   /SetVariables(
    Particles.Lama_Age_Scale=0.7
    Particles.Lama_Drag:NiagaraFloat=&MakeCustomFloatFromBool(Bool=$User.bIsFrontEndPreview, True Float=0.5, False Float=2)
    Particles.Lama_Force_Scale_A=100
    Particles.Lama_Force_Scale_X:NiagaraFloat=&MakeCustomFloatFromBool(Bool=$User.bIsFrontEndPreview, True Float=300, False Float=300)
    Particles.Lama_Force_Scale_Y:NiagaraFloat=&MakeCustomFloatFromBool(Bool=$User.bIsFrontEndPreview, True Float=300, False Float=300)
    Particles.Lama_LifeTime:NiagaraFloat=&RandomRangeFloat(Minimum=3, Maximum=5)
    Particles.Lama_Rotation_Rate:NiagaraFloat=&RandomRangeFloat(Minimum=0.5, Maximum=1.2)
    Particles.Lama_Scale:NiagaraFloat=&MakeCustomFloatFromBool(Bool=$User.bIsFrontEndPreview, True Float=0.35, False Float=0.25)
    Particles.Lama_Velocity_Speed:NiagaraFloat=&MakeCustomFloatFromBool(Bool=$User.bIsFrontEndPreview, True Float=100, False Float=250)
   ) {isSetParameters=true, module=SetVariables_#3}
   /SetVariables(Particles.Stick_Force_Scale_A=500, Particles.Stick_Force_Scale_X=500, Particles.Stick_Force_Scale_Y=500) {isSetParameters=true}
   &PartitionParticles()
   &InitializeParticle(
    Color Mode=::NewEnumerator0
    Lifetime=&Lerp_Float(
     A=&RandomRangeFloat(Minimum=1.4, Maximum=1.2)
     B=&RandomRangeFloat(Minimum=1.8, Maximum=2)
     Alpha=$User.ParaGlideLeanAlpha
    )
    Mesh Scale Mode=::NewEnumerator3
    Mesh Uniform Scale=&Lerp_Float(
     A=&RandomRangeFloat(Minimum=2, Maximum=3)
     B=&RandomRangeFloat(Minimum=2, Maximum=3)
     Alpha=$User.ParaGlideLeanAlpha
    )
    UsePositionOffset=false
    Write Material Random=true
   )
   &SocketLocation(
    Orientation Sampling=::NewEnumerator1
    SamplePosOffset=false
    Socket Index=&ModuloInt(A=&ReturnParticleIDAttributes(), B=2)
    Socket Reader=$System.Socket Reader
    Socket Sampling Mode=::NewEnumerator3
    VelocitySpeedLimit=false
   )
   &AddVelocity(
    Velocity=&SelectVectorFromArray(
     Direct Array Index=&Subtract_Integer(
      A:NiagaraInt32=1
      B:NiagaraInt32=&ModuloInt(A=&ReturnParticleIDAttributes(), B=2)
     )
     Vector Selection Array=$System.Axis
    )
    Velocity Speed Scale=100
   )
  ]
  particleUpdateScript [
   &ParticleState()
   &LinearForce(Force=&Add_Vector(
     A=&MakeVector(
      X=&Sine(
       Normalized Angle=&Multiply_Float(A=$Emitter.Age, B=0.5)
       Period=1
       Scale=$Particles.Stick_Force_Scale_X
       Bias=0
      )
      Y=&Cosine(
       Normalized Angle=&Multiply_Float(A=$Emitter.Age, B=0.5)
       Period=1
       Scale=$Particles.Stick_Force_Scale_Y
       Bias=0
      )
      Z=0
     )
     B=&Multiply_VectorByFloat(
      Vector=&SelectVectorFromArray(
       Direct Array Index=&Subtract_Integer(
        A:NiagaraInt32=1
        B:NiagaraInt32=&ModuloInt(A=&ReturnParticleIDAttributes(), B=2)
       )
       Vector Selection Array=$System.Axis
      )
      Float=$Particles.Stick_Force_Scale_A
     )
    ))
   &Drag(Drag=&Lerp_Float(A=6, B=2, Alpha=$User.ParaGlideLeanAlpha), Use Linear Drag=true, Use Rotational Drag=true)
   &UpdateMeshOrientation(
    Rotation Rate=&Lerp_Float(A=&RandomRangeFloat(Minimum=0.4, Maximum=0.8), B=2, Alpha=$User.ParaGlideLeanAlpha)
    Rotation Vector=&RandomVector(Vector Scale=1)
   )
   &ScaleMeshSize(Scale Factor=&VectorFromFloat(Value=&FloatFromCurve(FloatCurve=DataInterface() {propertyValues {Curve {
         defaultValue=0.9887
         keys [{interpMode=RCIM_Cubic, tangentMode=RCTM_User, time=0.5, value=1}, {arriveTangent=-3.1325, tangentMode=RCTM_User, time=1, value=0}]
        }}})))
   &SolveForcesAndVelocity(Clamp Velocity=false, Limit Acceleration=false)
  ]
  rendererClasses [&NiagaraMeshRendererProperties]
  renderers [{rendererClass=&NiagaraMeshRendererProperties, rendererIndex=0}]
 }
 Tornado_Swirl {
  emitterUpdateScript [
   &EmitterState(Life Cycle Mode=::NewEnumerator1, Loop Behavior=::NewEnumerator1, Loop Duration Mode=::NewEnumerator1, UseLoopDelay=false)
   &SpawnBurst_Instantaneous(Spawn Count=2, Use Spawn Probability=false, UseLoopCountLimit=false)
  ]
  particleSpawnScript [
   &InitializeParticle(
    Color=&FNTimeOfDayColor(
     Night Color=<rgba 0.5517 0.7251 1 0.7>
     Morning Color=<rgba 1 1 1 0.5>
     Day Color=<rgba 0.8 0.8 0.8 1>
     Evening Color=<rgba 1 0.8396 0.7176 1>
    )
    Lifetime=5
    Mesh Renderer Array Visibility Mode=::NewEnumerator1
    Mesh Renderer Info=DataInterface() {propertyValues {MeshRenderer=&self:Tornado_Swirl_0.NiagaraMeshRendererProperties_0}}
    Mesh Renderer Mesh Index:NiagaraInt32=&ReturnParticleIDAttributes()
    Mesh Scale Mode=::NewEnumerator3
    Mesh Uniform Scale=0.3
    UsePositionOffset=false
    Write Material Random=true
   )
   /SetVariables(Particles.RotationSpeed:NiagaraFloat=&Multiply_Float(
     A=1.5
     B=&Lerp_Float(
      A=1
      B=-1
      Alpha=&MakeFloatFromInt(Int=&ReturnParticleIDAttributes())
     )
    )) {isSetParameters=true}
  ]
  particleUpdateScript [
   &ParticleState(Kill Particles When Lifetime Has Elapsed=false)
   &ScaleMeshSize(Scale Factor=&Lerp_Vector(A=<1 1 1.35>, B=<1 1 2.5>, Alpha=$User.ParaGlideLeanAlpha))
   &SocketLocation(
    Orientation Sampling=::NewEnumerator1
    SamplePosOffset=false
    Socket Index=&ReturnParticleIDAttributes()
    Socket Reader=$System.Socket Reader
    Socket Sampling Mode=::NewEnumerator3
    VelocitySpeedLimit=false
   )
   &self:QuatRotTornado(Dir:NiagaraDataInterfaceArrayFloat3=$System.Axis, RotRate=15) {module=QuatRotTornado}
   &ScaleColor(Scale Alpha=&Lerp_Float(A=0.35, B=0.2, Alpha=$User.ParaGlideLeanAlpha), ScaleA=true, ScaleRGB=true)
  ]
  rendererClasses [&NiagaraMeshRendererProperties]
  renderers [
   {rendererClass=&NiagaraMeshRendererProperties, rendererIndex=0}
   {rendererClass=&NiagaraMeshRendererProperties, rendererIndex=1}
   {rendererClass=&NiagaraMeshRendererProperties, rendererIndex=2}
   {rendererClass=&NiagaraMeshRendererProperties, rendererIndex=3}
  ]
 }
}
summary {
 systemName=NS_Contrail_CentralGust
 emitters [
  {emitterName=ShockWave, rendererClasses [&NiagaraMeshRendererProperties]}
  {emitterName=ParticletoRibbon, rendererClasses [&NiagaraRibbonRendererProperties]}
  {emitterName=Sparks, rendererClasses [&NiagaraSpriteRendererProperties]}
  {emitterName=Tornado_Swirl, rendererClasses [&NiagaraMeshRendererProperties]}
  {emitterName=Leafs, rendererClasses [&NiagaraMeshRendererProperties]}
  {emitterName=Sticks, rendererClasses [&NiagaraMeshRendererProperties]}
  {emitterName=Lama, rendererClasses [&NiagaraMeshRendererProperties]}
 ]
 userVariables [
  User.bIsFrontEnd {type=NiagaraBool}
  User.bIsFrontEndPreview {type=NiagaraBool}
  User.ParaGlideLeanAlpha {type=NiagaraFloat, defaultValue=1, description=""}
 ]
}
